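; Weapon Blueprint for the LoudKiwi SMG that adds material-driven fire and reload effects on top of its parent class behavior.
; On firing, it sets the weapon meshes' power level to the magazine's fill ratio and timestamps a fire flash with game time.
; On reload start, a fixed delay precedes a timestamped reload flash and a reset of power to full, independent of actual ammo count.
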(event EventBeginPlay
  (|CallParentFunction))

(event EventTick
  (params (DeltaSeconds :type Float%20%28single-precision%29)))

(event EventOnPlayReloadFX
  (params (ReloadStage :type EFortReloadFXState%20Enum :default "ReloadStart"))
  (|CallParentFunction ReloadStage)
  ; Material Effects
  (switch Utilities|FlowControl|Switch|SwitchonEFortReloadFXState ReloadStage
    (:ReloadStart
      (Utilities|FlowControl|Delay 1.65)
      (Rendering|Material|SetScalarParameterValueonMaterials (Visual|GetWeaponMeshes) "LoudKiwi.BP_TimeReload" (Utilities|Time|GetGameTimeinSeconds)) ; Reload Flash
      (Rendering|Material|SetScalarParameterValueonMaterials (Visual|GetWeaponMeshes) "LoudKiwi.BP_PowerLevel" 1.0))))

(event EventOnPlayWeaponFireFX
  (params (bPersistentFire :type Boolean) (bSecondaryFire :type Boolean))
  ; Material Effects
  (|CallParentFunction bPersistentFire bSecondaryFire)
  (Rendering|Material|SetScalarParameterValueonMaterials (Visual|GetWeaponMeshes) "LoudKiwi.BP_PowerLevel" (/ (Math|Conversions|ToFloat(Integer) (FortWeapon|GetMagazineAmmoCount)) (Math|Conversions|ToFloat(Integer) (FortWeapon|GetBulletsPerClip))))
  (Rendering|Material|SetScalarParameterValueonMaterials (Visual|GetWeaponMeshes) "LoudKiwi.BP_TimeOnFire" (Utilities|Time|GetGameTimeinSeconds)) ; On Fire Flash
)

(fn ConstructionScript
  (|CallParentFunction))
