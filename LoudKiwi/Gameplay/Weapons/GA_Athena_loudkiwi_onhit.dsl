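; Gameplay ability for the LoudKiwi SMG that, on hit, builds explosion stacks on the struck target.
; CheckActorType selects structure, vehicle or player stacking and explosion effects from the target's tags.
; At the stack limit it clears stacks and, if enabled, applies the explosion; first stacks also await tag removal to apply it.
; Knockback, vehicle impulse, bullet-count and hit-info helpers exist, unused by the activation graph.
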(event Ability|EventActivateAbilityFromEvent
  (params (EventData :type Gameplay%20Event%20Data%20Structure%20%28by%20ref%29))
  (bind (_eventtag _instigator _target _optionalobject _optionalobject2 _contexthandle _instigatortags _targettags _eventmagnitude _targetdata) (Utilities|Struct|BreakGameplayEventData EventData))
  (bind _returnvalue (Ability|TargetData|GetHitResultfromTargetData _targetdata))
  (bind (_bblockinghit _binitialoverlap _time _distance _location _impactpoint _normal _impactnormal _physmat _hitactor) (Collision|BreakHitResult _returnvalue))
  (bind _returnvalue_2 (Ability|GetAbilitySystemComponent _hitactor))
  (bind _returnvalue_3 (Ability|GetActivatingPawn))
  (bind _returnvalue_4 (Ability|GetAbilitySystemComponent _returnvalue_3))
  (bind _hitactor_abilitysystemcomponent (Variables|Default|GetHitActor_AbilitySystemComponent))
  (|CallParentFunction EventData)
  (bind _asplayer_pawn_athena (Utilities|Casting|CastToPlayerPawn_Athena _instigator))
  (if (Tags|WeaponDataHasMatchingGameplayTag (Variables|/Script/FortniteGame.FortPawn|GetCurrentWeapon _asplayer_pawn_athena) "(TagName=\"Item.Weapon.Ranged.SMG.LoudKiwi\")")
    (Variables|Default|SetMaxStacks false)
    (if (FortWeapon|IsEquipped (Ability|GetCurrentSourceWeapon))
      (Utilities|IsValid _returnvalue_2
        (:Is%20Valid
          (Variables|Default|SetHitActor_AbilitySystemComponent _returnvalue_2)
          (CallFunction|CheckActorType)
          ; Check Stack count. If max stack do explosion and clear GEs, if not continue Stacking
          (bind _returnvalue_5 (GameplayEffects|GetActiveEffectswithAllTags (Variables|Default|GetHitActor_AbilitySystemComponent) "(GameplayTags=((TagName=\"LoudKiwi.ExplosionStacking\")))"))
          (for _array_element _returnvalue_5
            (bind _returnvalue_6 (Ability|GameplayEffect|GetActiveGameplayEffectStackCount _array_element))
            (bind _returnvalue_7 (Ability|GameplayEffect|GetActiveGameplayEffectStackLimitCount _array_element))
            (Variables|Default|SetMaxStacks (>= _returnvalue_6 _returnvalue_7))
            (Variables|Default|SetNotMAx (> _returnvalue_6 0)))
          (if (Variables|Default|GetMaxStacks)
            (Variables|Default|SetMaxStackDmg true)
            (GameplayEffects|RemoveActiveEffectswithTags _hitactor_abilitysystemcomponent "(GameplayTags=((TagName=\"LoudKiwi.ExplosionStacking\")))")
            (if (ScalableFloat|GetValueasBool (Variables|Default|GetInput))
              (GameplayEffects|ApplyGameplayEffectToTarget :self _returnvalue_4 :GameplayEffectClass (Variables|GEs|GetGE_MaxStack_Explosion) :Target _hitactor_abilitysystemcomponent :Level 0.0 :!label _label))
            (else
              (Ability|Tasks|ApplyGameplayEffectToHitResultsAsync _returnvalue_4 (Utilities|Array|MakeArray _returnvalue) (Variables|GEs|GetGE_Stacking) (stale _returnvalue_6) 1
                (:OnCompleted
                  (if (not (Variables|Default|GetNotMAx))
                    (Ability|Tasks|WaitGameplayTagRemoveFromActor _hitactor "(TagName=\"LoudKiwi.ExplosionStacking\")" true
                      (:Removed
                        (goto _label)))))))))
        (:Is%20Not%20Valid
          _)))
    (Ability|EndAbility)))

(event Custom|CheckActorType
  (bind _hitactor_abilitysystemcomponent (Variables|Default|GetHitActor_AbilitySystemComponent))
  ; CheckActorType and set appropriate GEs to apply
  (if (GameplayTags|HasAnyMatchingGameplayTags _hitactor_abilitysystemcomponent (Variables|Default|GetT_Structure))
    (Variables|GEs|SetGE_Stacking (Variables|GEs|GetGE%20Stacking_Structure))
    (Variables|GEs|SetGE_MaxStack_Explosion (Variables|GEs|GetGE%20Max%20Stack%20Explosion%20Structures))
    (elif (GameplayTags|HasAnyMatchingGameplayTags _hitactor_abilitysystemcomponent (Variables|Default|GetT_Vehicle))
      (Variables|GEs|SetGE_Stacking (Variables|GEs|GetGE%20Stacking_Vehicle))
      (Variables|GEs|SetGE_MaxStack_Explosion (Variables|GEs|GetGE%20Max%20Stack%20Explosion%20Vehicle))
      (else
        (Variables|GEs|SetGE_Stacking (Variables|GEs|GetGE%20Stacking_Player))
        (Variables|GEs|SetGE_MaxStack_Explosion (Variables|GEs|GetGE%20Max%20Stack%20Explosion%20Player))))))

(fn CalculateEnvKnock
  (bind _outactors (Collision|SphereOverlapActors (Variables|Default|GetHit%20Location) (ScalableFloat|GetValueatLevel (Variables|Default|GetHFKnockbackRadius)) (Variables|Default|GetPawnsFilterObjectType)))
  (for _array_element _outactors
    (bind _launchedpawn (Utilities|LaunchPawns (CallFunction|GetAvatarasFortPlayerPawn) (Utilities|Casting|CastToFortPawn _array_element) (* (Math|Vector|Normalize (- (Transformation|GetActorLocation (Utilities|Casting|CastToFortPawn _array_element)) (Variables|Default|GetHit%20Location))) (ScalableFloat|GetValueatLevel (Variables|Default|GetHFKnockbackIntensity))) true true true true true true))
    (bind _returnvalue (Spec|MakeSpecHandle(ByClass) (Variables|Default|GetGE_FallDamageImmunity) (Ability|GetAvatarActorfromActorInfo)))
    (GameplayEffects|ApplyGameplayEffectSpecToTarget (Ability|GetFortAbilitySystemComponent _launchedpawn) _returnvalue (Ability|GetFortAbilitySystemComponent _launchedpawn)))
  (return))

(fn CalculateVehicleImpulse
  (params (Object :type Actor%20Object%20Reference))
  (return (Math|Vector|MakeVector (.x (* (Math|Vector|Normalize (- (Transformation|GetActorLocation Object) (Variables|Default|GetHit%20Location))) (ScalableFloat|GetValueatLevel (Variables|Default|GetHFVehicleKnockbackIntensity)))) (.y (* (Math|Vector|Normalize (- (Transformation|GetActorLocation Object) (Variables|Default|GetHit%20Location))) (ScalableFloat|GetValueatLevel (Variables|Default|GetHFVehicleKnockbackIntensity)))) (+ (.z (* (Math|Vector|Normalize (- (Transformation|GetActorLocation Object) (Variables|Default|GetHit%20Location))) (ScalableFloat|GetValueatLevel (Variables|Default|GetHFVehicleKnockbackIntensity)))) (ScalableFloat|GetValueatLevel (Variables|Default|GetHFVehicleZOffsetKnockbackIntensity))))))

(fn CheckAndHandleBulletCount
  (bind _returnvalue (Ability|GetCurrentSourceWeapon))
  (bind _returnvalue_2 (FortWeapon|Ammo|HasInfiniteMagazineAmmo _returnvalue))
  (bind _returnvalue_3 (not _returnvalue_2))
  (bind _returnvalue_4 (FortWeapon|GetBulletsPerClip _returnvalue))
  ; If not using the infinite ammo config
  (if _returnvalue_3
    (Variables|Default|SetBulletCount (- (FortWeapon|GetBulletsPerClip (Ability|GetCurrentSourceWeapon)) (FortWeapon|GetLocalMagazineAmmoCount (Ability|GetCurrentSourceWeapon))))
    (elif (and _returnvalue_3 (== (+ (FortWeapon|GetLocalMagazineAmmoCount _returnvalue) 1) _returnvalue_4))
      (Variables|Default|SetBulletCount 0)
      ; Check if reached bullets per clip
      (if (>= (Variables|Default|GetBulletCount) _returnvalue_4) :!label _label
        (Variables|Default|SetBulletCount 1)
        (else
          (bind _result (Math|Integer|IncrementInt (Variables|Default|GetBulletCount))
            (:%20%20
              (Variables|Default|SetBulletCount _result)))))
      (else
        (goto _label))))
  (return))

(fn DirectHitLaunchHandle)

(fn EnvironmentalKnockback
  (CallFunction|CalculateEnvKnock))

(fn GetHitInfo
  (params (EventData :type Gameplay%20Event%20Data%20Structure))
  (bind (_eventtag _instigator _target _optionalobject _optionalobject2 _contexthandle _instigatortags _targettags _eventmagnitude _targetdata) (Utilities|Struct|BreakGameplayEventData EventData))
  (for _index (range (Ability|TargetData|GetDataCountfromTargetData _targetdata))
    (if (Collision|BreakHitResult (Ability|TargetData|GetHitResultfromTargetData _targetdata _index))
      (return true (.Location (Collision|BreakHitResult (Ability|TargetData|GetHitResultfromTargetData _targetdata _index))) (.HitActor (Collision|BreakHitResult (Ability|TargetData|GetHitResultfromTargetData _targetdata _index))))))
  (return false "0, 0, 0" 0))

(fn OnRep_MaxStackDmg)
