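; LoudKiwi weapon gameplay ability that tracks explosion stacking and applies a max-stack explosion effect.
; Activation calls the parent, then waits for removal of the ExplosionStacking tag on the activating pawn to apply that effect to the hit target.
; CheckActorType selects structure, vehicle or player effect variants by tag.
; Helpers parse hit data, track bullet count, compute vehicle impulse, and launch nearby pawns radially with fall-damage immunity.
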(event Ability|EventActivateAbility
  (bind _returnvalue (Ability|GetActivatingPawn))
  (|CallParentFunction)
  (Ability|Tasks|WaitGameplayTagRemoveFromActor _returnvalue "(TagName=\"LoudKiwi.ExplosionStacking\")"
    (:then
      (Utilities|FlowControl|Delay 3.0))
    (:Removed
      (GameplayEffects|ApplyGameplayEffectToTarget :self (Ability|GetAbilitySystemComponent _returnvalue) :GameplayEffectClass (Variables|GEs|GetGE_MaxStack_Explosion) :Target (Variables|Default|GetHitActor_AbilitySystemComponent) :Level 0.0)))
  (Ability|EndAbility))

(event Custom|CheckActorType
  (bind _hitactor_abilitysystemcomponent (Variables|Default|GetHitActor_AbilitySystemComponent))
  ; CheckActorType and set appropriate GEs to apply
  (if (GameplayTags|HasAnyMatchingGameplayTags _hitactor_abilitysystemcomponent (Variables|Default|GetT_Structure))
    (Variables|GEs|SetGE_Stacking (Variables|GEs|GetGE%20Stacking_Structure))
    (Variables|GEs|SetGE_MaxStack_Explosion (Variables|GEs|GetGE%20Max%20Stack%20Explosion%20Structures))
    (elif (GameplayTags|HasAnyMatchingGameplayTags _hitactor_abilitysystemcomponent (Variables|Default|GetT_Vehicle))
      (Variables|GEs|SetGE_Stacking (Variables|GEs|GetGE%20Stacking_Vehicle))
      (Variables|GEs|SetGE_MaxStack_Explosion (Variables|GEs|GetGE%20Max%20Stack%20Explosion%20Vehicle))
      (else
        (Variables|GEs|SetGE_Stacking (Variables|GEs|GetGE%20Stacking_Player))
        (Variables|GEs|SetGE_MaxStack_Explosion (Variables|GEs|GetGE%20Max%20Stack%20Explosion%20Player))))))

(fn CalculateEnvKnock
  (bind _outactors (Collision|SphereOverlapActors (Variables|Default|GetHit%20Location) (ScalableFloat|GetValueatLevel (Variables|Default|GetHFKnockbackRadius)) (Variables|Default|GetPawnsFilterObjectType)))
  (for _array_element _outactors
    (bind _launchedpawn (Utilities|LaunchPawns (CallFunction|GetAvatarasFortPlayerPawn) (Utilities|Casting|CastToFortPawn _array_element) (* (Math|Vector|Normalize (- (Transformation|GetActorLocation (Utilities|Casting|CastToFortPawn _array_element)) (Variables|Default|GetHit%20Location))) (ScalableFloat|GetValueatLevel (Variables|Default|GetHFKnockbackIntensity))) true true true true true true))
    (bind _returnvalue (Spec|MakeSpecHandle(ByClass) (Variables|Default|GetGE_FallDamageImmunity) (Ability|GetAvatarActorfromActorInfo)))
    (GameplayEffects|ApplyGameplayEffectSpecToTarget (Ability|GetFortAbilitySystemComponent _launchedpawn) _returnvalue (Ability|GetFortAbilitySystemComponent _launchedpawn)))
  (return))

(fn CalculateVehicleImpulse
  (params (Object :type Actor%20Object%20Reference))
  (return (Math|Vector|MakeVector (.x (* (Math|Vector|Normalize (- (Transformation|GetActorLocation Object) (Variables|Default|GetHit%20Location))) (ScalableFloat|GetValueatLevel (Variables|Default|GetHFVehicleKnockbackIntensity)))) (.y (* (Math|Vector|Normalize (- (Transformation|GetActorLocation Object) (Variables|Default|GetHit%20Location))) (ScalableFloat|GetValueatLevel (Variables|Default|GetHFVehicleKnockbackIntensity)))) (+ (.z (* (Math|Vector|Normalize (- (Transformation|GetActorLocation Object) (Variables|Default|GetHit%20Location))) (ScalableFloat|GetValueatLevel (Variables|Default|GetHFVehicleKnockbackIntensity)))) (ScalableFloat|GetValueatLevel (Variables|Default|GetHFVehicleZOffsetKnockbackIntensity))))))

(fn CheckAndHandleBulletCount
  (bind _returnvalue (Ability|GetCurrentSourceWeapon))
  (bind _returnvalue_2 (FortWeapon|Ammo|HasInfiniteMagazineAmmo _returnvalue))
  (bind _returnvalue_3 (not _returnvalue_2))
  (bind _returnvalue_4 (FortWeapon|GetBulletsPerClip _returnvalue))
  ; If not using the infinite ammo config
  (if _returnvalue_3
    (Variables|Default|SetBulletCount (- (FortWeapon|GetBulletsPerClip (Ability|GetCurrentSourceWeapon)) (FortWeapon|GetLocalMagazineAmmoCount (Ability|GetCurrentSourceWeapon))))
    (elif (and _returnvalue_3 (== (+ (FortWeapon|GetLocalMagazineAmmoCount _returnvalue) 1) _returnvalue_4))
      (Variables|Default|SetBulletCount 0)
      ; Check if reached bullets per clip
      (if (>= (Variables|Default|GetBulletCount) _returnvalue_4) :!label _label
        (Variables|Default|SetBulletCount 1)
        (else
          (bind _result (Math|Integer|IncrementInt (Variables|Default|GetBulletCount))
            (:%20%20
              (Variables|Default|SetBulletCount _result)))))
      (else
        (goto _label))))
  (return))

(fn DirectHitLaunchHandle)

(fn EnvironmentalKnockback
  (CallFunction|CalculateEnvKnock))

(fn GetHitInfo
  (params (EventData :type Gameplay%20Event%20Data%20Structure))
  (bind (_eventtag _instigator _target _optionalobject _optionalobject2 _contexthandle _instigatortags _targettags _eventmagnitude _targetdata) (Utilities|Struct|BreakGameplayEventData EventData))
  (for _index (range (Ability|TargetData|GetDataCountfromTargetData _targetdata))
    (if (Collision|BreakHitResult (Ability|TargetData|GetHitResultfromTargetData _targetdata _index))
      (return true (.Location (Collision|BreakHitResult (Ability|TargetData|GetHitResultfromTargetData _targetdata _index))) (.HitActor (Collision|BreakHitResult (Ability|TargetData|GetHitResultfromTargetData _targetdata _index))))))
  (return false "0, 0, 0" 0))
